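# Sound cue for the Soccer Traversal emote, classed as emote music.
# A Random node picks one of two equally weighted wave variations.
# It uses emote attenuation, a traversal emote property sheet, and traversal and global foley concurrency sets, and keeps playing when silent.
# A submix send also feeds it to the music audio-analysis submix.
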SoundCue /BRCosmetics/Sounds/Emotes/SoccerTraversal/Emote_SoccerTraversal
root=Random_0
nodes [
 {id=Random_0, children [WavePlayer_0,WavePlayer_1], properties {weights [1,1]}}
 {id=WavePlayer_0, properties {soundWaveAssetPtr=/Game/Athena/Sounds/Emotes/SoccerTraversal/Emote_SoccerTraversal_01}}
 {id=WavePlayer_1, properties {soundWaveAssetPtr=/Game/Athena/Sounds/Emotes/SoccerTraversal/Emote_SoccerTraversal_02}}
]
properties {
 attenuationSettings=&Fort_Emotes_Athena_Att
 audioPropertiesSheet=/BRCosmetics/Sounds/Settings/PropertySheets/Emotes/APS_Cosmetic_Emote_Traversal
 maxDistance=1600
 soundClassObject=&Fort_Music_Emote
 virtualizationMode=PlayWhenSilent
 volumeMultiplier=1.4
 concurrencyOverrides {maxCount=3}
 concurrencySet [/Game/Sounds/Concurrency_Rules/Athena_EmoteTraversal_Conc, /Game/Sounds/Concurrency_Rules/Athena_EmoteFoley_Global_Conc]
 soundSubmixSends [{minSendLevel=0, soundSubmix=&Fortnite_AudioAnalysis_Music}]
}
